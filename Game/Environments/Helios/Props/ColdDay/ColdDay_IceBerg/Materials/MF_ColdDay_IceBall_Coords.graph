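# Material function that samples one weighted frame from the ColdDay iceberg impostor atlas.
# Piramid Coords are scaled by Frames, shifted by Offset and clamped, then rounded up to select an atlas cell.
# UVs within that cell are divided by Frames to sample the texture.
# The result is multiplied by a two-axis tent weight from the fractional coordinate, which suggests summing offset calls to blend neighboring frames.
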MaterialFunction /Game/Environments/Helios/Props/ColdDay/ColdDay_IceBerg/Materials/MF_ColdDay_IceBall_Coords
graph {nodes [
  n1 = FunctionInput() {inputName=Piramid Coords, inputType=FunctionInput_Vector2, sortPriority=1}
  n2 = FunctionInput() {inputName=Frames, inputType=FunctionInput_Vector2, sortPriority=2}
  n3 = Multiply(A=@n2) {constB=0.25}
  n4 = FunctionInput(Preview=Constant2Vector()) {bUsePreviewValueAsDefault=true, inputName=Offset, inputType=FunctionInput_Vector2, sortPriority=3}
  n5 = Multiply(A=@n4) {constB=0.25}
  n6 = Multiply(A=@n3) {constB=-1}
  n7 = Multiply(A=@n4) {constB=0.5}
  n8 = Clamp(None=Add(A=Multiply(A=@n1, B=@n3), B=@n5) {constB=-0.25}, Min=Max(A=Add(A=@n6, B=@n7), B=@n6), Max=Min(A=Add(A=@n3, B=@n7), B=@n3)) {maxDefault=3.5, minDefault=-4}
  n9 = Constant2Vector() {g=-0.5, r=-0.5}
  n10 = Multiply(A=Add(A=Ceil(None=@n8), B=Subtract(A=@n9, B=@n5)) {constB=-0.25}) {constB=2}
  n11 = FunctionInput() {inputName=UVs, inputType=FunctionInput_Vector2}
  n12 = TextureSample(UVs=Divide(A=Add(A=@n10, B=@n11), B=@n2) {constB=8}) {samplerType=SAMPLERTYPE_LinearColor, texture=/Game/Environments/Helios/Props/ColdDay/ColdDay_IceBerg/Textures/T_ColdDay_Iceberg_Imposter}
  n13 = Add(A=Frac(None=@n8)) {constB=-0.5}
  n14 = OneMinus(None=Multiply(A=Abs(None=@n13)) {constB=2})
  n15 = ComponentMask(None=@n14) {r=true}
  n16 = ComponentMask(None=@n14) {g=true}
  FunctionOutput(None=Multiply(A=@n12.RGBA, B=Multiply(A=@n15, B=@n16))) {outputName=Texture}
 ]}
