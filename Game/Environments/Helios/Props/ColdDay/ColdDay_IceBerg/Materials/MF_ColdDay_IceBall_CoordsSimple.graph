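# Material function that samples the ColdDay iceberg imposter atlas at a single frame picked from 2D pyramid coordinates, judging by its input names.
# It scales the coordinates by half the frame counts, clamps them to the grid, then snaps with Ceil and a half offset.
# The UVs are added to that cell and divided by Frames for one linear-color texture sample output as RGBA, with no blending between frames.
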MaterialFunction /Game/Environments/Helios/Props/ColdDay/ColdDay_IceBerg/Materials/MF_ColdDay_IceBall_CoordsSimple
graph {nodes [
  n1 = FunctionInput() {inputName=Frames, inputType=FunctionInput_Vector2, sortPriority=2}
  n2 = Multiply(A=@n1) {constB=0.5}
  n3 = Multiply(A=FunctionInput() {inputName=Piramid Coords, inputType=FunctionInput_Vector2, sortPriority=1}, B=@n2)
  n4 = Add(A=@n2) {constB=-0.5}
  n5 = Multiply(A=@n4) {constB=-1}
  n6 = Add(A=Ceil(None=Clamp(None=@n3, Min=@n5, Max=@n4))) {constB=-0.5}
  n7 = FunctionInput() {inputName=UVs, inputType=FunctionInput_Vector2}
  n8 = TextureSample(UVs=Divide(A=Add(A=@n6, B=@n7), B=@n1) {constB=8}) {samplerType=SAMPLERTYPE_LinearColor, texture=/Game/Environments/Helios/Props/ColdDay/ColdDay_IceBerg/Textures/T_ColdDay_Iceberg_Imposter}
  FunctionOutput(None=@n8.RGBA) {outputName=Texture}
 ]}
