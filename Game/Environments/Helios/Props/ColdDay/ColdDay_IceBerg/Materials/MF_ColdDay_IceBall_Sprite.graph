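# Material function rendering the ColdDay iceberg's ice ball as a view-dependent impostor sprite, outputting Color and Alpha.
# It projects pixels onto a camera-facing plane through the offset pivot for UVs, and maps local view direction to square pyramid coordinates for frame lookup.
# High and Epic sum four offset frame samples; lower qualities and mobile use one simpler sample.
# Color gets tint and alpha-derived glow; square-masked alpha gets contrast.
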MaterialFunction /Game/Environments/Helios/Props/ColdDay/ColdDay_IceBerg/Materials/MF_ColdDay_IceBall_Sprite
graph {nodes [
  n1 = NamedRerouteUsage()
  n2 = NamedRerouteUsage()
  n3 = ScalarParameter() {defaultValue=12, group=Sprite, parameterName=Sprite Frames}
  n4 = MaterialFunctionCall!(UVs=@n1, Piramid Coords=@n2, Frames=@n3) {materialFunction=&/MF_ColdDay_IceBall_CoordsSimple}
  n5 = Constant2Vector() {g=-1, r=-1}
  n6 = Constant2Vector() {g=1, r=1}
  n7 = Constant2Vector() {g=1, r=-1}
  n8 = Constant2Vector() {g=-1, r=1}
  n9 = Add(
   A=Add(
    A=MaterialFunctionCall!(UVs=@n1, Piramid Coords=@n2, Frames=@n3, Offset=@n5) {materialFunction=&/MF_ColdDay_IceBall_Coords}.Texture
    B=MaterialFunctionCall(UVs=@n1, Piramid Coords=@n2, Frames=@n3, Offset=@n6) {materialFunction=&MF_ColdDay_IceBall_Coords}.Texture
   )
   B=Add(
    A=MaterialFunctionCall(UVs=@n1, Piramid Coords=@n2, Frames=@n3, Offset=@n7) {materialFunction=&MF_ColdDay_IceBall_Coords}.Texture
    B=MaterialFunctionCall(UVs=@n1, Piramid Coords=@n2, Frames=@n3, Offset=@n8) {materialFunction=&MF_ColdDay_IceBall_Coords}.Texture
   )
  )
  n10 = ShadingPathSwitch(Default=QualitySwitch(Default=@n4.Texture, Low=@n4.Texture, High=@n9, Medium=@n4.Texture, Epic=@n9), Mobile=@n4.Texture)
  n11 = ComponentMask(None=@n10) {b=true, g=true, r=true}
  n12 = Multiply(A=ComponentMask(None=@n10) {a=true}, B=NamedRerouteUsage())
  n13 = OneMinus(None=@n12)
  n14 = Multiply(A=Multiply(A=@n13, B=@n13) {constB=6}, B=ScalarParameter() {defaultValue=5, group=Sprite, parameterName=Sprite Outer Glow})
  FunctionOutput(None=Multiply(A=VectorParameter() {group=Sprite, parameterName=Sprite Color Multiplier, defaultValue=<rgba 0.0208 0.6339 1 1>}.RGB, B=Add(A=@n11, B=@n14))) {outputName=Color}
  n15 = ScalarParameter() {defaultValue=8, group=Sprite, parameterName=Sprite Alpha Contrast}
  FunctionOutput(None=Saturate(None=MaterialFunctionCall(In=@n12, Contrast=@n15) {materialFunction=&CheapContrast}.Result)) {outputName=Alpha, sortPriority=1}
  n16 = Multiply(
   A=Transform(None=NamedRerouteUsage()) {transformSourceType=TRANSFORMSOURCE_World, transformType=TRANSFORM_Local}
   B=MaterialFunctionCall() {materialFunction=&ObjectScale}.Scale XYZ
  )
  n17 = ComponentMask(None=@n16) {g=true, r=true}
  n18 = Abs(None=@n17)
  n19 = ComponentMask(None=@n18) {r=true}
  n20 = ComponentMask(None=@n18) {g=true}
  n21 = Multiply(A=ComponentMask(None=@n16) {b=true}) {constB=-1}
  n22 = Multiply(A=Divide(A=@n17, B=Max(A=@n19, B=@n20)), B=OneMinus(None=Saturate(None=@n21)))
  NamedRerouteDeclaration(None=AppendVector(A=ComponentMask(None=@n22) {g=true}, B=ComponentMask(None=@n22) {r=true})) {name=Piramid Vector, nodeColor=<rgba 0.4062 0 1 1>}
  n23 = MaterialFunctionCall() {materialFunction=&ObjectPivotPoint}
  n24 = CameraPositionWS()
  n25 = Normalize(VectorInput=Subtract(A=@n23.Object Pivot Location, B=@n24))
  n26 = Normalize(VectorInput=CrossProduct(A=Constant3Vector() {constant=<rgba 0 0 1 0>}, B=@n25))
  n27 = TruncateLWC!(None=WorldPosition().XYZ)
  n28 = Normalize(VectorInput=Subtract(A=@n27, B=@n24))
  n29 = Distance(A=@n23.Object Pivot Location, B=@n24)
  n30 = DotProduct(A=@n28, B=NamedRerouteDeclaration(None=@n25) {name=Direction Vector, nodeColor=<rgba 1 0 0.1641 1>})
  n31 = Multiply(A=@n28, B=Subtract(A=Divide(A=@n29, B=@n30), B=Distance(A=@n27, B=@n24)))
  n32 = Add(
   A=Subtract(A=Add(A=@n27, B=@n31), B=@n23.Object Pivot Location)
   B=Transform(None=VectorParameter() {parameterName=Sprite World Offset, defaultValue=<rgba 50 0 -170 1>}.RGB) {transformSourceType=TRANSFORMSOURCE_Local}
  ) {constB=0}
  n33 = AppendVector(A=DotProduct(A=@n26, B=@n32), B=DotProduct(A=CrossProduct(A=@n26, B=@n25), B=@n32))
  n34 = MaterialFunctionCall() {materialFunction=&ObjectScale}
  n35 = Min(A=@n34.Scale X, B=@n34.Scale X)
  n36 = Multiply(A=VertexInterpolator(VS=Min(A=@n35, B=@n34.Scale Z)).PS, B=ScalarParameter() {defaultValue=600, group=Sprite, parameterName=Sprite Scale})
  n37 = Min(A=Max(A=Divide(A=@n33, B=@n36) {constB=400}) {constB=-0.5}) {constB=0.5}
  n38 = ScalarParameter() {defaultValue=12, group=Sprite, parameterName=Sprite Frames}
  NamedRerouteDeclaration(None=Add(A=@n37, B=Multiply(A=@n38) {constB=0.5})) {name=UVs, nodeColor=<rgba 0 1 0.6719 1>}
  n39 = Round(None=Add(A=Abs(None=@n37)) {constB=0.005})
  n40 = ComponentMask(None=@n39) {r=true}
  n41 = ComponentMask(None=@n39) {g=true}
  NamedRerouteDeclaration(None=OneMinus(None=Max(A=@n40, B=@n41))) {name=Square Mask, nodeColor=<rgba 0.0312 0 1 1>}
 ]}
